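# Physics asset for the WavyMan male medium body skin mesh, providing physics bodies for its belly and four small valve bones.
# Capsules cover the pelvis, two spine bones, left lower arm and right upper arm as anchors for those dyn_ bones.
# The belly hangs off spine_02 with moderate limits, carrying one valve; the other valves attach tightly to pelvis and arms.
# The spine_02-to-pelvis joint is angularly free.
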PhysicsAsset /BRCosmetics/Characters/Player/Male/Medium/Bodies/M_MED_WavyMan/Meshes/M_MED_Wavy_Man_Skin_Physics
bodies {
 dyn_belly [{shapeType=Capsule, length=3.4881, radius=8.5298, center=<8.0427 0 -0.0001>, rotation=<rot 90 0 0>}]
 dyn_valve_belly [{shapeType=Capsule, length=3.508, radius=1.6177, center=<2.0664 -0.0001 0>, rotation=<rot 90 0 0>}]
 dyn_valve_hip [{shapeType=Capsule, length=3.3486, radius=1.6623, center=<2.1522 -0.0001 0.0001>, rotation=<rot 57.493 89.9999 89.9999>}]
 dyn_valve_l [{shapeType=Capsule, length=3.7083, radius=1.6898, center=<2.6754 -0.0002 0>, rotation=<rot 90 0 0>}]
 dyn_valve_r [{shapeType=Capsule, length=2.9423, radius=1.6871, center=<1.047 0 0>, rotation=<rot 37.1218 89.9997 89.9997>}]
 lowerarm_l [{shapeType=Capsule, length=12, radius=5.4, center=<8 0 0.5>, rotation=<rot 90 0 0>}]
 pelvis [{shapeType=Capsule, length=11.9426, radius=8.7716}]
 spine_02 [{shapeType=Capsule, length=16.3922, radius=5.9387, rotation=<rot 0 -90 0>}]
 spine_03 [{shapeType=Capsule, length=9.6064, radius=13.5522, center=<15.6546 1.2523 -0.248>, rotation=<rot 0 -61.9141 -89.0689>}]
 upperarm_r [{shapeType=Capsule, length=19.9697, radius=5.8152, center=<-10.3326 0.0001 -0.3564>, rotation=<rot 85.3004 0 0>}]
}
constraints [
 {bone1Name=spine_03, bone2Name=spine_02, swing1LimitDegrees=45, swing2LimitDegrees=45, twistLimitDegrees=45}
 {bone1Name=lowerarm_l, bone2Name=spine_03, swing1LimitDegrees=45, swing2LimitDegrees=45, twistLimitDegrees=45}
 {bone1Name=upperarm_r, bone2Name=spine_03, swing1LimitDegrees=45, swing2LimitDegrees=45, twistLimitDegrees=45}
 {bone1Name=dyn_belly, bone2Name=spine_02, swing1LimitDegrees=25, swing2LimitDegrees=15, twistLimitDegrees=4}
 {bone1Name=dyn_valve_belly, bone2Name=dyn_belly, swing1LimitDegrees=5, swing2LimitDegrees=5, twistLimitDegrees=1}
 {
  bone1Name=spine_02
  bone2Name=pelvis
  swing1LimitDegrees=45
  swing1Motion=Free
  swing2LimitDegrees=45
  swing2Motion=Free
  twistLimitDegrees=45
  twistMotion=Free
 }
 {bone1Name=dyn_valve_l, bone2Name=lowerarm_l, swing1LimitDegrees=5, swing2LimitDegrees=5, twistLimitDegrees=1}
 {bone1Name=dyn_valve_r, bone2Name=upperarm_r, swing1LimitDegrees=5, swing2LimitDegrees=5, twistLimitDegrees=1}
 {bone1Name=dyn_valve_hip, bone2Name=pelvis, swing1LimitDegrees=5, swing2LimitDegrees=5, twistLimitDegrees=1}
]
