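# Physics asset for the face accessory mesh part of the male medium WavyMan player body, covering a two-strand topknot.
# Capsule bodies sit on the head and on three bones in each of the topknot A and B chains.
# Chain A hangs from the head and chain B branches off the first A bone.
# Swing and twist limits bound each joint, with the B branch root held much tighter in swing than the other joints.
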PhysicsAsset /BRCosmetics/Characters/Player/Male/Medium/Bodies/M_MED_WavyMan/Meshes/Parts/M_MED_Wavy_Man_FaceAcc_Physics
bodies {
 dyn_topknotA_01 [{shapeType=Capsule, length=4.6379, radius=2.5591, center=<-0.6842 0 -0.0623>, rotation=<rot 78.4851 -179.9999 -179.9998>}]
 dyn_topknotA_02 [{shapeType=Capsule, length=2.6384, radius=1.1895, center=<2.0243 -0.0003 0>, rotation=<rot 90 0 0>}]
 dyn_topknotA_03 [{shapeType=Capsule, length=2.5674, radius=1.2234, center=<1.1196 0.0004 0>, rotation=<rot 90 0 0>}]
 dyn_topknotB_01 [{shapeType=Capsule, length=2.9033, radius=1.3023, center=<2.2993 -0.0002 0>, rotation=<rot 90 0 0>}]
 dyn_topknotB_02 [{shapeType=Capsule, length=1.3482, radius=1.3636, center=<1.4759 0.0001 0>, rotation=<rot 90 0 0>}]
 dyn_topknotB_03 [{shapeType=Capsule, length=2.1629, radius=1.7423, center=<2.2091 0.0002 0>, rotation=<rot 90 0 0>}]
 head [{shapeType=Capsule, length=14.7931, radius=8.4489, center=<6.8223 2.4098 0.0001>, rotation=<rot 90 0 0>}]
}
constraints [
 {bone1Name=dyn_topknotA_01, bone2Name=head, swing1LimitDegrees=35, swing2LimitDegrees=35, twistLimitDegrees=5}
 {bone1Name=dyn_topknotA_02, bone2Name=dyn_topknotA_01, swing1LimitDegrees=30, swing2LimitDegrees=30, twistLimitDegrees=8}
 {bone1Name=dyn_topknotA_03, bone2Name=dyn_topknotA_02, swing1LimitDegrees=30, swing2LimitDegrees=30, twistLimitDegrees=8}
 {bone1Name=dyn_topknotB_01, bone2Name=dyn_topknotA_01, swing1LimitDegrees=10, swing2LimitDegrees=10, twistLimitDegrees=5}
 {bone1Name=dyn_topknotB_02, bone2Name=dyn_topknotB_01, swing1LimitDegrees=30, swing2LimitDegrees=30, twistLimitDegrees=8}
 {bone1Name=dyn_topknotB_03, bone2Name=dyn_topknotB_02, swing1LimitDegrees=30, swing2LimitDegrees=30, twistLimitDegrees=10}
]
